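# Physics asset for the hammer bones on the F_MED_RoseForm female body cosmetic, modeling a hammer attached at the right thigh.
# A tiny master body links to the thigh and roots a three-capsule chain with per-joint swing limits and zero-degree twist limits.
# Capsules on the right thigh and calf represent the leg, and the master-to-thigh and calf-to-thigh joints leave swing and twist free.
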PhysicsAsset /BRCosmetics/Characters/Player/Female/Medium/Bodies/F_MED_RoseForm/Meshes/F_MED_RoseForm_Hammer_Physics
bodies {
 calf_r [{shapeType=Capsule, length=31.0707, radius=5.9773, shapeName=calf_r_capsule, center=<17.2331 2.3795 0>, rotation=<rot 90 80.7735 73.6237>}]
 dyn_hammer_01 [{shapeType=Capsule, length=16.8392, radius=1.8497, shapeName=dyn_hammer_01_capsule, center=<9.3912 0 0>, rotation=<rot 90 0 0>}]
 dyn_hammer_02 [{shapeType=Capsule, length=0, radius=2.3205, shapeName=dyn_hammer_02_capsule, center=<1.8436 0 0>}]
 dyn_hammer_03 [{
   shapeType=Capsule
   length=8.5925
   radius=1.1922
   shapeName=dyn_hammer_03_capsule
   center=<5.1006 0.0009 0.1094>
   rotation=<rot -88.7715 0.0098 0>
  }]
 dyn_hammer_MSTR [{shapeType=Capsule, length=0.5, radius=0.5, shapeName=dyn_hammer_MSTR_capsule}]
 thigh_r [{shapeType=Capsule, length=34.0389, radius=5.6209, shapeName=thigh_r_capsule, center=<25.5776 0 0>, rotation=<rot 90 89.4327 92.7655>}]
}
constraints [
 {bone1Name=dyn_hammer_01, bone2Name=dyn_hammer_MSTR, swing1LimitDegrees=15, swing2LimitDegrees=20, twistLimitDegrees=0}
 {bone1Name=dyn_hammer_02, bone2Name=dyn_hammer_01, swing1LimitDegrees=0, swing2LimitDegrees=40, twistLimitDegrees=0}
 {bone1Name=dyn_hammer_03, bone2Name=dyn_hammer_02, swing1LimitDegrees=60, swing2LimitDegrees=20, twistLimitDegrees=0}
 {
  bone1Name=dyn_hammer_MSTR
  bone2Name=thigh_r
  swing1LimitDegrees=45
  swing1Motion=Free
  swing2LimitDegrees=45
  swing2Motion=Free
  twistLimitDegrees=45
  twistMotion=Free
 }
 {
  bone1Name=calf_r
  bone2Name=thigh_r
  swing1LimitDegrees=45
  swing1Motion=Free
  swing2LimitDegrees=45
  swing2Motion=Free
  twistLimitDegrees=45
  twistMotion=Free
 }
]
