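# Physics asset for the F_MED_RoseForm female player body cosmetic, covering its back strap, overall flap and left pouch bones.
# Capsules on the pelvis and spine_02–05, plus a pelvis box, form the torso shapes that the accessory bodies attach to.
# The six-link strap chain and single flap link hang from master bones on spine_03 and spine_02, with swing limits and zero twist.
# The boxed pouch joins the pelvis with free angular motion, as do both masters and the spine links.
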PhysicsAsset /BRCosmetics/Characters/Player/Female/Medium/Bodies/F_MED_RoseForm/Meshes/F_MED_RoseForm_Physics
bodies {
 dyn_backStrap_01 [{shapeType=Capsule, length=1.3303, radius=1.5881, shapeName=dyn_backStrap_01_capsule, center=<1 0 0>, rotation=<rot 90 0 0>}]
 dyn_backStrap_02 [{shapeType=Capsule, length=3.3811, radius=1.3597, shapeName=dyn_backStrap_02_capsule, center=<1 0 0>, rotation=<rot 90 0 0>}]
 dyn_backStrap_03 [{shapeType=Capsule, length=5.0673, radius=1.3597, shapeName=dyn_backStrap_03_capsule, center=<2.0132 0.4752 0>, rotation=<rot 90 0 0>}]
 dyn_backStrap_04 [{shapeType=Capsule, length=5.0673, radius=1.3597, shapeName=dyn_backStrap_04_capsule, center=<2.0132 0.8 0>, rotation=<rot 90 0 0>}]
 dyn_backStrap_05 [{shapeType=Capsule, length=5.0673, radius=1.3597, shapeName=dyn_backStrap_05_capsule, center=<2.0132 0.8 0>, rotation=<rot 90 0 0>}]
 dyn_backStrap_06 [{shapeType=Capsule, length=5.0673, radius=1.3597, shapeName=dyn_backStrap_06_capsule, center=<2.0132 0.8 0>, rotation=<rot 90 0 0>}]
 dyn_backStrap_MSTR [{shapeType=Capsule, length=2, radius=1, shapeName=dyn_backStrap_MSTR_capsule, rotation=<rot 90 0 0>}]
 dyn_overallFlap_c_01 [{shapeType=Capsule, length=4.7064, radius=0.7188, shapeName=dyn_overallFlap_c_01_capsule, center=<8.0876 0 0>, rotation=<rot 90 0 0>}]
 dyn_overallFlap_c_MSTR [{shapeType=Capsule, length=0.505, radius=0.505, shapeName=dyn_overallFlap_c_MSTR_capsule}]
 dyn_pouch_l_01 [{
   shapeType=Box
   extentX=12.0467
   extentY=9.6952
   extentZ=7.5
   shapeName=dyn_pouch_l_01_box
   center=<1.1166 0.6515 0>
   rotation=<rot 0 -4.7045 0>
  }]
 pelvis [
  {shapeType=Capsule, length=7.5479, radius=11.9283, shapeName=pelvis_capsule, center=<-4.1045 -0.6281 0>}
  {
   shapeType=Box
   extentX=7.4713
   extentY=7.8107
   extentZ=12.6124
   shapeName=pelvis_box
   center=<2.9179 -10.974 0.4029>
   rotation=<rot 13.5479 20.8823 -6.6798>
  }
 ]
 spine_02 [{shapeType=Capsule, length=4.7135, radius=8.7873, shapeName=spine_02_capsule, center=<-6.4986 0 0>}]
 spine_03 [{shapeType=Capsule, length=2.6693, radius=8.2108, shapeName=spine_03_capsule, center=<-2.9162 1.7283 0>}]
 spine_04 [{shapeType=Capsule, length=1.6438, radius=8.3562, shapeName=spine_04_capsule}]
 spine_05 [{shapeType=Capsule, length=1.4417, radius=8.5583, shapeName=spine_05_capsule}]
}
constraints [
 {
  bone1Name=spine_04
  bone2Name=spine_03
  swing1LimitDegrees=45
  swing1Motion=Free
  swing2LimitDegrees=45
  swing2Motion=Free
  twistLimitDegrees=45
  twistMotion=Free
 }
 {
  bone1Name=spine_05
  bone2Name=spine_04
  swing1LimitDegrees=45
  swing1Motion=Free
  swing2LimitDegrees=45
  swing2Motion=Free
  twistLimitDegrees=45
  twistMotion=Free
 }
 {
  bone1Name=spine_03
  bone2Name=spine_02
  swing1LimitDegrees=45
  swing1Motion=Free
  swing2LimitDegrees=45
  swing2Motion=Free
  twistLimitDegrees=45
  twistMotion=Free
 }
 {
  bone1Name=spine_02
  bone2Name=pelvis
  swing1LimitDegrees=45
  swing1Motion=Free
  swing2LimitDegrees=45
  swing2Motion=Free
  twistLimitDegrees=45
  twistMotion=Free
 }
 {bone1Name=dyn_backStrap_06, bone2Name=dyn_backStrap_05, swing1LimitDegrees=20, swing2LimitDegrees=20, twistLimitDegrees=0}
 {
  bone1Name=dyn_backStrap_MSTR
  bone2Name=spine_03
  swing1LimitDegrees=45
  swing1Motion=Free
  swing2LimitDegrees=45
  swing2Motion=Free
  twistLimitDegrees=45
  twistMotion=Free
 }
 {bone1Name=dyn_backStrap_01, bone2Name=dyn_backStrap_MSTR, swing1LimitDegrees=20, swing2LimitDegrees=20, twistLimitDegrees=0}
 {bone1Name=dyn_backStrap_02, bone2Name=dyn_backStrap_01, swing1LimitDegrees=20, swing2LimitDegrees=20, twistLimitDegrees=0}
 {bone1Name=dyn_backStrap_03, bone2Name=dyn_backStrap_02, swing1LimitDegrees=20, swing2LimitDegrees=20, twistLimitDegrees=0}
 {bone1Name=dyn_backStrap_04, bone2Name=dyn_backStrap_03, swing1LimitDegrees=20, swing2LimitDegrees=20, twistLimitDegrees=0}
 {bone1Name=dyn_backStrap_05, bone2Name=dyn_backStrap_04, swing1LimitDegrees=20, swing2LimitDegrees=20, twistLimitDegrees=0}
 {
  bone1Name=dyn_overallFlap_c_MSTR
  bone2Name=spine_02
  swing1LimitDegrees=45
  swing1Motion=Free
  swing2LimitDegrees=45
  swing2Motion=Free
  twistLimitDegrees=45
  twistMotion=Free
 }
 {bone1Name=dyn_overallFlap_c_01, bone2Name=dyn_overallFlap_c_MSTR, swing1LimitDegrees=45, swing2LimitDegrees=5, twistLimitDegrees=0}
 {
  bone1Name=dyn_pouch_l_01
  bone2Name=pelvis
  swing1LimitDegrees=45
  swing1Motion=Free
  swing2LimitDegrees=45
  swing2Motion=Free
  twistLimitDegrees=45
  twistMotion=Free
 }
]
