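# Animation Blueprint for the RoseForm female body part, adding secondary motion to straps, pouch, hammer and other accessories.
# It runs a deform Control Rig on the copied parent pose, then corrects bones using constraints, joint-rotation alphas and skydive, parachute and melee states.
# RigidBody physics follows, using a separate asset when the matching RoseForm backpack is equipped and damped acceleration for other back blings.
# A final thigh-relative simulation handles the hammer.
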AnimBlueprint /BRCosmetics/Characters/Player/Female/Medium/Bodies/F_MED_RoseForm/Meshes/F_MED_RoseForm_AnimBP
target_skeleton=&/F_MED_RoseForm_Skeleton
graphs [
 AnimGraph {nodes [
   n1 = "Constraint - dyn_hammer_MSTR (PPPP)"!(
    ComponentPose=Local To Component(LocalPose=Use cached pose 'overallCache'())
    Alpha=1
    ConstraintWeights_0=0.022
    ConstraintWeights_1:float=0.035
    ConstraintWeights_2:float=0.559
    ConstraintWeights_3:float=0.384
   ) -> (Pose) {
    alphaBoolBlend {blendOption=Linear}
    boneToModify {boneName=dyn_hammer_MSTR}
    constraintSetup [
     {targetBone {boneName=deform_groin_r}}
     {targetBone {boneName=pelvis}}
     {targetBone {boneName=thigh_r}}
     {targetBone {boneName=thigh_twist_01_r}}
    ]
    constraintWeights [1,1,1,1]
   }
   n2 = "Constraint - dyn_pouch_r_MSTR (PPPPPP)"!(
    ComponentPose=@n1
    Alpha=1
    ConstraintWeights_0=0.007
    ConstraintWeights_1:float=0.029
    ConstraintWeights_2:float=0.038
    ConstraintWeights_3:float=0.07
    ConstraintWeights_4:float=0.325
    ConstraintWeights_5:float=0.531
   ) -> (Pose) {
    alphaBoolBlend {blendOption=Linear}
    boneToModify {boneName=dyn_pouch_r_MSTR}
    constraintSetup [
     {targetBone {boneName=deform_glute_r}}
     {targetBone {boneName=deform_groin_r}}
     {targetBone {boneName=dyn_gluteBelt_r_01}}
     {targetBone {boneName=pelvis}}
     {targetBone {boneName=thigh_r}}
     {targetBone {boneName=thigh_twist_01_r}}
    ]
    constraintWeights [1,1,1,1,1,1]
   }
   n3 = RigidBody!(ComponentPose=@n2, Alpha, OverrideWorldGravity:Vector=Get Gravity Override(), ExternalForce) -> (Pose) {
    bOverrideWorldGravity=true
    overridePhysicsAsset=&/F_MED_RoseForm_Hammer_Physics
    simulationSpace=BaseBoneSpace
    alphaBoolBlend {blendOption=Linear}
    baseBoneRef {boneName=thigh_r}
    componentAppliedLinearAccClamp=<6000 6000 6000>
    componentLinearAccScale=<1.5 0.8 0.8>
    componentLinearVelScale=<1.5 0.8 0.8>
    overrideWorldGravity {z=-980}
   }
   Output Pose(Result=Component To Local(ComponentPose=@n3)) {layerGroup=None}
   n4 = Control Rig(Source=Copy Pose From Mesh() {bUseAttachedParent=true}) {
    alphaBoolBlend {blendOption=Linear}
    controlRigAssetReference {blueprintRigClass=&/F_MED_RoseForm_DeformRig_ControlRig.F_MED_RoseForm_DeformRig_ControlRig_C}
   }
   Save cached pose 'CRCache'(Pose=@n4) {cacheName=CRCache}
   n5 = Get accScaleVals() -> (accScaleVals:Vector)
   n6 = RigidBody!(
    ComponentPose=Local To Component(LocalPose=Use cached pose 'overallMechanicsCache'())
    Alpha
    OverrideWorldGravity=Get Gravity Override()
    ExternalForce
    ComponentLinearAccScale:Vector=@n5
    ComponentLinearVelScale:Vector=@n5
   ) -> (Pose) {
    bOverrideWorldGravity=true
    overridePhysicsAsset=&/F_MED_RoseForm_Physics
    simulationSpace=BaseBoneSpace
    alphaBoolBlend {blendOption=Linear}
    baseBoneRef {boneName=spine_01}
    componentAppliedLinearAccClamp=<6000 6000 6000>
    componentLinearAccScale=<0.5 0.5 0.5>
    componentLinearVelScale=<0.5 0.5 0.5>
    overrideWorldGravity {z=-980}
   }
   n7 = Get accScaleVals()
   n8 = RigidBody(
    ComponentPose=Local To Component(LocalPose=Use cached pose 'overallMechanicsCache'())
    OverrideWorldGravity=Get Gravity Override()
    ComponentLinearAccScale=@n7
    ComponentLinearVelScale=@n7
   ) {
    bOverrideWorldGravity=true
    overridePhysicsAsset=&/F_MED_RoseForm_BackBling_Physics
    simulationSpace=BaseBoneSpace
    alphaBoolBlend {blendOption=Linear}
    baseBoneRef {boneName=spine_01}
    componentAppliedLinearAccClamp=<6000 6000 6000>
    componentLinearAccScale=<0.5 0.5 0.5>
    componentLinearVelScale=<0.5 0.5 0.5>
    overrideWorldGravity {z=-980}
   }
   Save cached pose 'overallCache'(Pose=Blend Poses by bool(
     BlendPose_0=Component To Local(ComponentPose=@n8)
     BlendPose_1=Component To Local(ComponentPose=@n6)
     bActiveValue=Get isCharBackPack() -> (isCharBackPack:bool)
    ) {blendType=Linear}) {cacheName=overallCache}
   n9 = "Transform (Modify) Bone - Bone: dyn_gluteBelt_r_01"(ComponentPose=Local To Component(LocalPose=Use cached pose 'CRCache'()), Alpha=1, Translation="-2,-2,-1", Rotation="0,0,0", Scale="1,1,1") {
    rotationMode=BMM_Additive
    rotationSpace=BCS_BoneSpace
    translationMode=BMM_Additive
    translationSpace=BCS_BoneSpace
    alphaBoolBlend {blendOption=Linear}
    boneToModify {boneName=dyn_gluteBelt_r_01}
    translation=<-2 -2 -1>
   }
   n10 = "Transform (Modify) Bone - Bone: dyn_gluteBelt_r_01"(ComponentPose=@n9, Alpha=Get thigh Rot Y() -> (thigh Rot Y:float), Translation="-3,-5,0", Rotation="25,0,-10", Scale="1,1,1") {
    rotationMode=BMM_Additive
    rotationSpace=BCS_BoneSpace
    translationMode=BMM_Additive
    translationSpace=BCS_BoneSpace
    alphaBoolBlend {blendOption=Linear}
    boneToModify {boneName=dyn_gluteBelt_r_01}
    rotation {pitch=25, roll=-10}
    translation=<-3 -5>
   }
   n11 = "Transform (Modify) Bone - Bone: dyn_shoulderPad_r_01"!(ComponentPose=@n10, bAlphaBoolEnabled:bool=Get bIsParachuteOpened(), Translation="-3,0,0", Rotation="0,0,0", Scale="1,1,1") -> (Pose) {
    alphaInputType=Bool
    rotationMode=BMM_Additive
    rotationSpace=BCS_BoneSpace
    translationMode=BMM_Additive
    translationSpace=BCS_BoneSpace
    alphaBoolBlend {blendOption=Linear}
    boneToModify {boneName=dyn_shoulderPad_r_01}
    translation {x=-3}
   }
   n12 = "Constraint - dyn_backStrap_aim (PPPPP)"!(
    ComponentPose=@n11
    Alpha=1
    ConstraintWeights_0=0.159
    ConstraintWeights_1:float=0.092
    ConstraintWeights_2:float=0.158
    ConstraintWeights_3:float=0.521
    ConstraintWeights_4:float=0.07
   ) -> (Pose) {
    alphaBoolBlend {blendOption=Linear}
    boneToModify {boneName=dyn_backStrap_aim}
    constraintSetup [
     {targetBone {boneName=clavicle_r}}
     {targetBone {boneName=deform_upperarm_bk_r}}
     {targetBone {boneName=spine_04}}
     {targetBone {boneName=spine_05}}
     {targetBone {boneName=upperarm_twist_01_r}}
    ]
    constraintWeights [1,1,1,1,1]
   }
   n13 = "Constraint - dyn_backStrap_MSTR (PPPPPP)"!(
    ComponentPose=@n12
    Alpha=1
    ConstraintWeights_0=0.15
    ConstraintWeights_1:float=0.119
    ConstraintWeights_2:float=0.492
    ConstraintWeights_3:float=0.08
    ConstraintWeights_4:float=0.133
    ConstraintWeights_5:float=0.026
   ) -> (Pose) {
    alphaBoolBlend {blendOption=Linear}
    boneToModify {boneName=dyn_backStrap_MSTR}
    constraintSetup [
     {targetBone {boneName=pelvis}}
     {targetBone {boneName=spine_01}}
     {targetBone {boneName=spine_02}}
     {targetBone {boneName=spine_03}}
     {targetBone {boneName=spine_04}}
     {targetBone {boneName=spine_05}}
    ]
    constraintWeights [1,1,1,1,1,1]
   }
   Save cached pose 'thighBagCache'(Pose=Component To Local(ComponentPose=@n13)) {cacheName=thighBagCache}
   n14 = "Transform (Modify) Bone - Bone: dyn_backStrap_MSTR"(
    ComponentPose=Local To Component(LocalPose=Use cached pose 'thighBagCache'())
    Alpha=Get clavicle Up() -> (clavicle Up:float)
    Translation="1,-1,1"
    Rotation="-3,0,0"
    Scale="1,1,1"
   ) {
    rotationMode=BMM_Additive
    rotationSpace=BCS_BoneSpace
    translationMode=BMM_Additive
    translationSpace=BCS_BoneSpace
    alphaBoolBlend {blendOption=Linear}
    alphaScaleBiasClamp {bInterpResult=true}
    boneToModify {boneName=dyn_backStrap_MSTR}
    rotation {pitch=-3}
    translation=<1 -1 1>
   }
   n15 = "Transform (Modify) Bone - Bone: dyn_backStrap_MSTR"!(ComponentPose=@n14, bAlphaBoolEnabled:bool=Get bIsPlayingMeleeAnim(), Translation="0,-1,0", Rotation="0,0,0", Scale="1,1,1") -> (Pose) {
    alphaInputType=Bool
    rotationMode=BMM_Additive
    rotationSpace=BCS_BoneSpace
    translationMode=BMM_Additive
    translationSpace=BCS_BoneSpace
    alphaBoolBlend {blendInTime=1, blendOption=Linear, blendOutTime=1}
    boneToModify {boneName=dyn_backStrap_MSTR}
    translation {y=-1}
   }
   n16 = Get bIsParachuteOpened()
   n17 = "Transform (Modify) Bone - Bone: dyn_backStrap_MSTR"(ComponentPose=@n15, bAlphaBoolEnabled=@n16, Translation="-3,-2,0", Rotation="0,0,0", Scale="1,1,1") {
    alphaInputType=Bool
    rotationMode=BMM_Additive
    rotationSpace=BCS_BoneSpace
    translationMode=BMM_Additive
    translationSpace=BCS_BoneSpace
    alphaBoolBlend {blendInTime=1, blendOption=Linear, blendOutTime=1}
    boneToModify {boneName=dyn_backStrap_MSTR}
    translation=<-3 -2>
   }
   n18 = "Transform (Modify) Bone - Bone: dyn_backStrap_aim"!(ComponentPose=@n17, bAlphaBoolEnabled:bool=@n16, Translation="1,-2,0", Rotation="0,0,0", Scale="1,1,1") -> (Pose) {
    alphaInputType=Bool
    rotationMode=BMM_Additive
    translationMode=BMM_Additive
    alphaBoolBlend {blendOption=Linear}
    boneToModify {boneName=dyn_backStrap_aim}
    translation=<1 -2>
   }
   n19 = Get bIsSkydiveFloating()
   n20 = "Transform (Modify) Bone - Bone: dyn_backStrap_aim"(ComponentPose=@n18, bAlphaBoolEnabled=@n19, Translation="1,2,4", Rotation="0,0,0", Scale="1,1,1") {
    alphaInputType=Bool
    rotationMode=BMM_Additive
    translationMode=BMM_Additive
    alphaBoolBlend {blendOption=Linear}
    boneToModify {boneName=dyn_backStrap_aim}
    translation=<1 2 4>
   }
   n21 = "Transform (Modify) Bone - Bone: dyn_backStrap_MSTR"(ComponentPose=@n20, bAlphaBoolEnabled=@n19, Translation="0,-2,0", Rotation="0,0,0", Scale="1,1,1") {
    alphaInputType=Bool
    rotationMode=BMM_Additive
    rotationSpace=BCS_BoneSpace
    translationMode=BMM_Additive
    translationSpace=BCS_BoneSpace
    alphaBoolBlend {blendInTime=1, blendOption=Linear, blendOutTime=1}
    boneToModify {boneName=dyn_backStrap_MSTR}
    translation {y=-2}
   }
   n22 = "Transform (Modify) Bone - Bone: dyn_backStrap_aim"(ComponentPose=@n21, bAlphaBoolEnabled=Get bIsSkydiveDiveMode(), Translation="1,2,4", Rotation="0,0,0", Scale="1,1,1") {
    alphaInputType=Bool
    rotationMode=BMM_Additive
    translationMode=BMM_Additive
    alphaBoolBlend {blendOption=Linear}
    boneToModify {boneName=dyn_backStrap_aim}
    translation=<1 2 4>
   }
   n23 = "Transform (Modify) Bone - Bone: dyn_backStrap_aim"!(ComponentPose=@n22, Alpha:float=Get clavicle Bk() -> (clavicle Bk:double), Translation="0.5,-2,0", Rotation="0,0,0", Scale="1,1,1") -> (Pose) {
    rotationMode=BMM_Additive
    rotationSpace=BCS_BoneSpace
    translationMode=BMM_Additive
    translationSpace=BCS_BoneSpace
    alphaBoolBlend {blendOption=Linear}
    boneToModify {boneName=dyn_backStrap_aim}
    translation=<0.5 -2>
   }
   Save cached pose 'overallMechanicsCache'(Pose=Component To Local(ComponentPose=@n23)) {cacheName=overallMechanicsCache}
  ]}
 EventGraph {nodes [
   n1 = Event BlueprintUpdateAnimation()
   > n2 = Sequence!(execute) -> (then_0, then_1, then_2:Exec, then_3:Exec, then_4:Exec, then_5:Exec, then_6:Exec, then_7:Exec)
   > n3 = Gravity Override State Function DEPRECATED(CharacterPartAnimInstance=Self-Reference(), GameTimeInSeconds=@n1.DeltaTimeX, GravityOverrideParams=Get Gravity Override Params())
   > Set Gravity Override(Gravity Override=@n3)
   Set BodyMesh(execute=@n2.then_1, BodyMesh=GetPartSkeletalMeshComponent())
   Set clavicle Bk(
    execute=@n2.then_6
    clavicle Bk:double=GetAlphaFromReferenceTransformDelta(
     SkeletalMeshComponent=Get BodyMesh() -> (BodyMesh:Skeletal Mesh Component Object Reference)
     BoneName=clavicle_r
     BaseName=spine_05
     InputRange=(X=0,Y=-20)
     TwistAxis="1,0,0"
    )
   ) -> (Output_Get:double)
   Set Alpha Output(
    execute=@n2.then_7
    Alpha Output:float=GetAlphaFromReferenceTransformDelta(
     SkeletalMeshComponent=Get BodyMesh()
     BoneName=upperarm_r
     BaseName=clavicle_r
     DeltaComponentType=EulerX
     InputRange=(X=0,Y=15)
     TwistAxis="1,0,0"
    )
   ) -> (Output_Get:float)
   n4 = Get isCharBackPack()
   n5 = CheckForBackbling()
   n6 = Branch(
    execute=@n2.then_4
    Condition=AND Boolean(A=OR Boolean!(A=@n5, B=@n5.bIsBackblingACape, C:bool=@n5.bIsBackblingLong) -> (ReturnValue), B=NOT Boolean(A=@n4))
   )
   > Set accScaleVals(accScaleVals:Vector="0.02,0.02,0.1") -> (Output_Get:Vector)
   Set accScaleVals(execute=@n6.else, accScaleVals="0.5,0.5,0.5")
   Set clavicle Up(
    execute=@n2.then_3
    clavicle Up:float=GetAlphaFromReferenceTransformDelta(
     SkeletalMeshComponent=Get BodyMesh()
     BoneName=clavicle_r
     BaseName=spine_05
     DeltaComponentType=EulerY
     InputRange=(X=0,Y=40)
     TwistAxis="1,0,0"
    )
   ) -> (Output_Get:float)
   Set thigh Rot Y(
    execute=@n2.then_2
    thigh Rot Y:float=GetAlphaFromReferenceTransformDelta(SkeletalMeshComponent=Get BodyMesh(), BoneName=thigh_r, BaseName=pelvis, InputRange=(X=15,Y=-60), TwistAxis="1,0,0")
   ) -> (Output_Get:float)
   n7 = Cast To FortPlayerPawn(execute=@n2.then_5, Object=GetOwningActor())
   > Is Valid(InputObject=Get RoseFormBackPack() -> (RoseFormBackPack:Skeletal Mesh Object Reference))
   > n8 = Set Backpack(Backpack:Athena Backpack Item Definition Object Reference=GetItemDefinitionTypeFromLoadoutWithActor!(
     Actor:Actor Object Reference=@n7
     LoadoutSlot:Gameplay Tag Structure=(TagName="LoadoutSlot.Backpack")
     CosmeticType:Athena Cosmetic Item Definition Class Reference=/Script/FortniteGame.AthenaBackpackItemDefinition
    ) -> (OutCosmeticItemDef:Athena Backpack Item Definition Object Reference)) -> (Output_Get:Athena Backpack Item Definition Object Reference)
   > n9 = Branch(Condition="Equal (Object)"(A=@n8, B=/BRCosmetics/Athena/Items/Cosmetics/Backpacks/Backpack_RoseForm))
   > Set isCharBackPack(isCharBackPack:bool=true) -> (Output_Get:bool)
   Set isCharBackPack(execute=@n9.else, isCharBackPack=false)
  ]}
]
variables [
 accScaleVals {type=Vector}
 Alpha Output {type=number}
 As Fort Player Pawn {type=FortPlayerPawn}
 Backpack {type=AthenaBackpackItemDefinition, default=null}
 BodyMesh {type=SkeletalMeshComponent}
 clavicle Bk {type=number}
 clavicle Up {type=number}
 Gravity Override {type=Vector}
 Gravity Override Params {type=GravityOverrideParamsStruct}
 isBackblingOn {type=boolean}
 isCharBackPack {type=boolean}
 RoseFormBackPack {type=SkeletalMesh, default=/BRCosmetics/Accessories/FORT_Backpacks/Backpack_F_MED_RoseForm/Meshes/F_MED_RoseForm_Pack}
 roseFormPack {type=AthenaBackpackItem, default=null}
 thigh Rot Y {type=number}
]
