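# Physics asset for the skydive variant of the RoseForm female medium face accessory mesh, covering a ponytail and left and right hair strands.
# Capsule bodies cover the upper spine, neck, head, clavicles and right upper arm only, joined by free-rotation constraints.
# Head-attached MSTR anchors root two-segment side strands and a five-segment ponytail, whose links use per-joint swing limits and zero twist.
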PhysicsAsset /BRCosmetics/Characters/Player/Female/Medium/Bodies/F_MED_RoseForm/Meshes/Parts/F_MED_RoseForm_FaceAcc_Skydive_Physics
bodies {
 clavicle_l [{shapeType=Capsule, length=7.8434, radius=4.9253, shapeName=clavicle_l_capsule, rotation=<rot 90 0 0>}]
 clavicle_r [{
   shapeType=Capsule
   length=10.7156
   radius=6.8215
   shapeName=clavicle_r_capsule
   center=<-0.1356 1.6178 1.0199>
   rotation=<rot 81.4876 -0.0006 -0.0006>
  }]
 dyn_hairMesh_l_01 [{shapeType=Capsule, length=2.8702, radius=0.9163, shapeName=dyn_hairMesh_l_01_capsule, center=<1.3865 0.5 0>, rotation=<rot 90 0 0>}]
 dyn_hairMesh_l_02 [{shapeType=Capsule, length=2.8702, radius=0.9163, shapeName=dyn_hairMesh_l_02_capsule, center=<1.3865 0.5 0>, rotation=<rot 90 0 0>}]
 dyn_hairMesh_l_MSTR [{shapeType=Capsule, length=0.5, radius=0.5, shapeName=dyn_hairMesh_l_MSTR_capsule}]
 dyn_hairMesh_r_01 [{shapeType=Capsule, length=2.65, radius=1.0676, shapeName=dyn_hairMesh_r_01_capsule, center=<1.6104 -0.5 0>, rotation=<rot 90 0 0>}]
 dyn_hairMesh_r_02 [{shapeType=Capsule, length=2.65, radius=1.0676, shapeName=dyn_hairMesh_r_02_capsule, center=<1.6104 -0.5 0>, rotation=<rot 90 0 0>}]
 dyn_hairMesh_r_MSTR [{shapeType=Capsule, length=0.5, radius=0.5, shapeName=dyn_hairMesh_r_MSTR_capsule}]
 dyn_ponyTail_01 [{
   shapeType=Capsule
   length=3.4188
   radius=2.1914
   shapeName=dyn_ponyTail_01_capsule
   center=<2.6673 0 -0.0783>
   rotation=<rot 82.389 0.0001 0.0001>
  }]
 dyn_ponyTail_02 [{shapeType=Capsule, length=4.0205, radius=1.7457, shapeName=dyn_ponyTail_02_capsule, center=<2.3954 0 -0.4>, rotation=<rot 78.4723 0 0>}]
 dyn_ponyTail_03 [{
   shapeType=Capsule
   length=5.2438
   radius=2.1089
   shapeName=dyn_ponyTail_03_capsule
   center=<2.4576 0 1>
   rotation=<rot 80.9554 179.9999 179.9999>
  }]
 dyn_ponyTail_04 [{
   shapeType=Capsule
   length=3.6752
   radius=3.464
   shapeName=dyn_ponyTail_04_capsule
   center=<3.5305 0 2.3>
   rotation=<rot 87.7695 -0.0001 -0.0001>
  }]
 dyn_ponyTail_05 [{
   shapeType=Capsule
   length=5.5578
   radius=3.464
   shapeName=dyn_ponyTail_04_capsule
   center=<3.5305 0 1.6>
   rotation=<rot 87.0659 -0.0003 -0.0003>
  }]
 dyn_ponyTail_MSTR [{shapeType=Capsule, length=0.5, radius=0.5, shapeName=dyn_ponyTail_MSTR_capsule}]
 head [{shapeType=Capsule, length=8.0238, radius=5.8096, shapeName=head_capsule, center=<2.3945 2.4683 0>, rotation=<rot 90 180 39.1725>}]
 neck_01 [{shapeType=Capsule, length=0.1105, radius=4.8895, shapeName=neck_01_capsule, rotation=<rot 90 0 0>}]
 neck_02 [{shapeType=Capsule, length=4.0028, radius=4.0667, shapeName=neck_02_capsule, center=<1.8778 0 0>, rotation=<rot 90 0 0>}]
 spine_04 [
  {shapeType=Capsule, length=13.6529, radius=10.805, shapeName=spine_05_capsule, center=<2.4656 -0.9064 0>}
  {shapeType=Capsule, length=9.6627, radius=10.3589, shapeName=spine_04_capsule, center=<-1.5426 3.9847 0>}
 ]
 spine_05 [{shapeType=Capsule, length=14.9731, radius=9.4848, shapeName=spine_05_capsule, center=<1.1213 -2.8501 0>}]
 upperarm_r [{shapeType=Capsule, length=18.5254, radius=4.9253, shapeName=upperarm_r_capsule, center=<-9.3245 0 -1.134>, rotation=<rot 90 0 0>}]
}
constraints [
 {
  bone1Name=spine_05
  bone2Name=spine_04
  swing1LimitDegrees=45
  swing1Motion=Free
  swing2LimitDegrees=45
  swing2Motion=Free
  twistLimitDegrees=45
  twistMotion=Free
 }
 {
  bone1Name=head
  bone2Name=neck_02
  swing1LimitDegrees=45
  swing1Motion=Free
  swing2LimitDegrees=45
  swing2Motion=Free
  twistLimitDegrees=45
  twistMotion=Free
 }
 {
  bone1Name=dyn_hairMesh_l_MSTR
  bone2Name=head
  swing1LimitDegrees=45
  swing1Motion=Free
  swing2LimitDegrees=45
  swing2Motion=Free
  twistLimitDegrees=45
  twistMotion=Free
 }
 {bone1Name=dyn_hairMesh_l_01, bone2Name=dyn_hairMesh_l_MSTR, swing1LimitDegrees=20, swing2LimitDegrees=10, twistLimitDegrees=0}
 {bone1Name=dyn_hairMesh_l_02, bone2Name=dyn_hairMesh_l_01, swing1LimitDegrees=20, swing2LimitDegrees=10, twistLimitDegrees=0}
 {
  bone1Name=dyn_hairMesh_r_MSTR
  bone2Name=head
  swing1LimitDegrees=45
  swing1Motion=Free
  swing2LimitDegrees=45
  swing2Motion=Free
  twistLimitDegrees=45
  twistMotion=Free
 }
 {bone1Name=dyn_hairMesh_r_01, bone2Name=dyn_hairMesh_r_MSTR, swing1LimitDegrees=20, swing2LimitDegrees=10, twistLimitDegrees=0}
 {bone1Name=dyn_hairMesh_r_02, bone2Name=dyn_hairMesh_r_01, swing1LimitDegrees=20, swing2LimitDegrees=10, twistLimitDegrees=0}
 {
  bone1Name=dyn_ponyTail_MSTR
  bone2Name=head
  swing1LimitDegrees=45
  swing1Motion=Free
  swing2LimitDegrees=45
  swing2Motion=Free
  twistLimitDegrees=45
  twistMotion=Free
 }
 {bone1Name=dyn_ponyTail_01, bone2Name=dyn_ponyTail_MSTR, swing1LimitDegrees=30, swing2LimitDegrees=15, twistLimitDegrees=0}
 {bone1Name=dyn_ponyTail_02, bone2Name=dyn_ponyTail_01, swing1LimitDegrees=40, swing2LimitDegrees=40, twistLimitDegrees=0}
 {bone1Name=dyn_ponyTail_03, bone2Name=dyn_ponyTail_02, swing1LimitDegrees=20, swing2LimitDegrees=30, twistLimitDegrees=0}
 {bone1Name=dyn_ponyTail_04, bone2Name=dyn_ponyTail_03, swing1LimitDegrees=20, swing2LimitDegrees=15, twistLimitDegrees=0}
 {
  bone1Name=neck_02
  bone2Name=neck_01
  swing1LimitDegrees=45
  swing1Motion=Free
  swing2LimitDegrees=45
  swing2Motion=Free
  twistLimitDegrees=45
  twistMotion=Free
 }
 {
  bone1Name=neck_01
  bone2Name=spine_05
  swing1LimitDegrees=45
  swing1Motion=Free
  swing2LimitDegrees=45
  swing2Motion=Free
  twistLimitDegrees=45
  twistMotion=Free
 }
 {
  bone1Name=clavicle_l
  bone2Name=spine_05
  swing1LimitDegrees=45
  swing1Motion=Free
  swing2LimitDegrees=45
  swing2Motion=Free
  twistLimitDegrees=45
  twistMotion=Free
 }
 {
  bone1Name=clavicle_r
  bone2Name=spine_05
  swing1LimitDegrees=45
  swing1Motion=Free
  swing2LimitDegrees=45
  swing2Motion=Free
  twistLimitDegrees=45
  twistMotion=Free
 }
 {
  bone1Name=upperarm_r
  bone2Name=clavicle_r
  swing1LimitDegrees=45
  swing1Motion=Free
  swing2LimitDegrees=45
  swing2Motion=Free
  twistLimitDegrees=45
  twistMotion=Free
 }
 {bone1Name=dyn_ponyTail_05, bone2Name=dyn_ponyTail_04, swing1LimitDegrees=20, swing2LimitDegrees=15, twistLimitDegrees=0}
]
