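# Post-process-style AnimBP for the RoseForm female character's FaceAcc part, driving ponytail physics.
# Parent and body-mesh poses are layered, ponytail bones are nudged under parachute, then a bool blend selects normal or skydive RigidBody simulation in spine_05 space.
# The event graph updates gravity override, acceleration scale, head/clavicle alphas and backbling-aware hair-lift flags.
# The hair-lift bone nodes those flags enable are disconnected from the output pose.
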AnimBlueprint /BRCosmetics/Characters/Player/Female/Medium/Bodies/F_MED_RoseForm/Meshes/Parts/F_MED_RoseForm_FaceAcc_AnimBP
target_skeleton=&/F_MED_RoseForm_FaceAcc_Skeleton
graphs [
 AnimGraph {nodes [
   n1 = Get liftHairParachute() -> (liftHairParachute:bool)
   "Transform (Modify) Bone - Bone: dyn_ponyTail_05"!(ComponentPose, bAlphaBoolEnabled:bool=@n1, Translation="0,0,0", Rotation="-19.431,-18.310,40.612", Scale="1,1,1") -> (Pose) {
    alphaInputType=Bool
    rotationMode=BMM_Additive
    translationMode=BMM_Additive
    alphaBoolBlend {blendOption=Linear}
    boneToModify {boneName=dyn_ponyTail_05}
    rotation=<rot -19.431 -18.31 40.612>
   }
   n2 = Get liftHairSkydive() -> (liftHairSkydive:bool)
   n3 = "Transform (Modify) Bone - Bone: dyn_ponyTail_MSTR"!(ComponentPose, bAlphaBoolEnabled:bool=@n2, Translation="0,0,0", Rotation="-35.849,78.100,5.619", Scale="1,1,1") -> (Pose) {
    alphaInputType=Bool
    rotationMode=BMM_Additive
    translationMode=BMM_Additive
    alphaBoolBlend {blendOption=Linear}
    boneToModify {boneName=dyn_ponyTail_MSTR}
    rotation=<rot -35.849 78.1 5.619>
   }
   n4 = "Transform (Modify) Bone - Bone: dyn_ponyTail_01"!(ComponentPose=@n3, bAlphaBoolEnabled:bool=@n2, Translation="0,0,0", Rotation="0,0,0", Scale="1,1,1") -> (Pose) {
    alphaInputType=Bool
    rotationMode=BMM_Additive
    translationMode=BMM_Additive
    alphaBoolBlend {blendOption=Linear}
    boneToModify {boneName=dyn_ponyTail_01}
   }
   n5 = "Transform (Modify) Bone - Bone: dyn_ponyTail_02"!(ComponentPose=@n4, bAlphaBoolEnabled:bool=@n2, Translation="0,0,0", Rotation="0.000,-4.872,33.917", Scale="1,1,1") -> (Pose) {
    alphaInputType=Bool
    rotationMode=BMM_Additive
    translationMode=BMM_Additive
    alphaBoolBlend {blendOption=Linear}
    boneToModify {boneName=dyn_ponyTail_02}
    rotation {roll=33.917, yaw=-4.872}
   }
   n6 = "Transform (Modify) Bone - Bone: dyn_ponyTail_03"!(ComponentPose=@n5, bAlphaBoolEnabled:bool=@n2, Translation="0,0,1", Rotation="-74.726,84.786,-61.790", Scale="1,1,1") -> (Pose) {
    alphaInputType=Bool
    rotationMode=BMM_Additive
    translationMode=BMM_Additive
    alphaBoolBlend {blendOption=Linear}
    boneToModify {boneName=dyn_ponyTail_03}
    rotation=<rot -74.726 84.786 -61.79>
    translation {z=1}
   }
   n7 = "Transform (Modify) Bone - Bone: dyn_ponyTail_04"!(ComponentPose=@n6, bAlphaBoolEnabled:bool=@n2, Translation="0,0,0", Rotation="-26.773,-5.300,34.135", Scale="1,1,1") -> (Pose) {
    alphaInputType=Bool
    rotationMode=BMM_Additive
    translationMode=BMM_Additive
    alphaBoolBlend {blendOption=Linear}
    boneToModify {boneName=dyn_ponyTail_04}
    rotation=<rot -26.773 -5.3 34.135>
   }
   "Transform (Modify) Bone - Bone: dyn_ponyTail_05"(ComponentPose=@n7, bAlphaBoolEnabled=@n2, Translation="0,0,0", Rotation="16.643,22.551,29.124", Scale="1,1,1") {
    alphaInputType=Bool
    rotationMode=BMM_Additive
    translationMode=BMM_Additive
    alphaBoolBlend {blendOption=Linear}
    boneToModify {boneName=dyn_ponyTail_05}
    rotation=<rot 16.643 22.551 29.124>
   }
   "Transform (Modify) Bone - Bone: dyn_ponyTail_02"(bAlphaBoolEnabled=@n1, Translation="0,0,0", Rotation="-35.034,18.558,32.745", Scale="1,1,1") {
    alphaInputType=Bool
    rotationMode=BMM_Additive
    translationMode=BMM_Additive
    alphaBoolBlend {blendOption=Linear}
    boneToModify {boneName=dyn_ponyTail_02}
    rotation=<rot -35.034 18.558 32.745>
   }
   n8 = Get linearAccScale() -> (linearAccScale:Vector)
   n9 = RigidBody!(
    ComponentPose=Local To Component(LocalPose=Use cached pose 'mechanicsCache'())
    Alpha
    OverrideWorldGravity:Vector=Get Gravity Override()
    ExternalForce
    ComponentLinearAccScale:Vector=@n8
    ComponentLinearVelScale:Vector=@n8
   ) -> (Pose) {
    bAllowPoseScalePropagation=false
    bClampLinearTranslationLimitToRefPose=true
    bOverrideWorldGravity=true
    overridePhysicsAsset=&/F_MED_RoseForm_FaceAcc_Physics1
    simulationSpace=BaseBoneSpace
    alphaBoolBlend {blendOption=Linear}
    baseBoneRef {boneName=spine_05}
    componentAppliedLinearAccClamp=<6000 6000 6000>
    componentLinearAccScale=<0.4 0.3 0.3>
    componentLinearVelScale=<0.4 0.3 0.3>
   }
   n10 = RigidBody(
    ComponentPose=Local To Component(LocalPose=Use cached pose 'mechanicsCache'())
    OverrideWorldGravity="0,0,-980"
    ComponentLinearAccScale="0.02,0.02,0.02"
    ComponentLinearVelScale="0.02,0.02,0.02"
   ) {
    bAllowPoseScalePropagation=false
    bClampLinearTranslationLimitToRefPose=true
    bOverrideWorldGravity=true
    overridePhysicsAsset=&/F_MED_RoseForm_FaceAcc_Skydive_Physics
    simulationSpace=BaseBoneSpace
    alphaBoolBlend {blendOption=Linear}
    baseBoneRef {boneName=spine_05}
    componentAppliedLinearAccClamp=<6000 6000 6000>
    componentLinearAccScale=<0.4 0.3 0.3>
    componentLinearVelScale=<0.4 0.3 0.3>
   }
   Output Pose(Result=Blend Poses by bool(BlendPose_0=Component To Local(ComponentPose=@n10), BlendPose_1=Component To Local(ComponentPose=@n9), bActiveValue=Get bIsSkyDiving()) {blendType=Linear}) {layerGroup=None}
   n11 = Get bIsParachuteOpened()
   n12 = "Transform (Modify) Bone - Bone: dyn_ponyTail_MSTR"(
    ComponentPose=Local To Component(LocalPose=Use cached pose 'bodyPose'())
    bAlphaBoolEnabled=@n11
    Translation="0,0,2"
    Rotation="0,0,0"
    Scale="1,1,1"
   ) {
    alphaInputType=Bool
    rotationMode=BMM_Additive
    translationMode=BMM_Additive
    alphaBoolBlend {blendOption=Linear}
    boneToModify {boneName=dyn_ponyTail_MSTR}
    translation {z=2}
   }
   n13 = "Transform (Modify) Bone - Bone: dyn_ponyTail_03"(ComponentPose=@n12, bAlphaBoolEnabled=@n11, Translation="0,0,2", Rotation="0,0,0", Scale="1,1,1") {
    alphaInputType=Bool
    rotationMode=BMM_Additive
    rotationSpace=BCS_BoneSpace
    translationMode=BMM_Additive
    translationSpace=BCS_BoneSpace
    alphaBoolBlend {blendOption=Linear}
    boneToModify {boneName=dyn_ponyTail_03}
    translation {z=2}
   }
   n14 = "Transform (Modify) Bone - Bone: clavicle_l"!(ComponentPose=@n13, bAlphaBoolEnabled:bool=@n11, Translation="0,0,0", Rotation="0,0,0", Scale="1,1,1") -> (Pose) {
    alphaInputType=Bool
    rotationMode=BMM_Additive
    translationMode=BMM_Additive
    alphaBoolBlend {blendOption=Linear}
    boneToModify {boneName=clavicle_l}
   }
   n15 = "Transform (Modify) Bone - Bone: clavicle_r"!(ComponentPose=@n14, bAlphaBoolEnabled:bool=@n11, Translation="0,0,0", Rotation="0,0,0", Scale="1,1,1") -> (Pose) {
    alphaInputType=Bool
    rotationMode=BMM_Additive
    translationMode=BMM_Additive
    alphaBoolBlend {blendOption=Linear}
    boneToModify {boneName=clavicle_r}
   }
   Save cached pose 'mechanicsCache'(Pose=Component To Local(ComponentPose=@n15)) {cacheName=mechanicsCache}
   n16 = Layered blend per bone(
    BasePose=Copy Pose From Mesh() {bCopyCurves=true, bUseAttachedParent=true}
    BlendPoses_0=Copy Pose From Mesh(SourceMeshComponent=Get BodyMesh() -> (BodyMesh:Skeletal Mesh Component Object Reference))
   ) {layerSetup [{branchFilters [
       {boneName=upperarm_l}
       {boneName=upperarm_r}
       {boneName=lowerarm_l}
       {boneName=lowerarm_r}
       {boneName=thigh_l}
       {boneName=thigh_r}
       {boneName=calf_l}
       {boneName=calf_r}
      ]}]}
   Save cached pose 'bodyPose'(Pose=@n16) {cacheName=bodyPose}
   Get Gravity Override()
   Get linearAccScale()
  ]}
 EventGraph {nodes [
   ResetDynamics(InTeleportType=TeleportPhysics)
   > ResetDynamics(InTeleportType=ResetPhysics)
   n1 = Event BlueprintUpdateAnimation()
   > n2 = Sequence!(execute) -> (then_0, then_1, then_2:Exec, then_3:Exec, then_4:Exec, then_5:Exec)
   > n3 = Gravity Override State Function DEPRECATED(CharacterPartAnimInstance=Self-Reference(), GameTimeInSeconds=@n1.DeltaTimeX, GravityOverrideParams=Get Gravity Override Params())
   > Set Gravity Override(Gravity Override=@n3)
   n4 = Get bIsParachuteOpened()
   n5 = CheckForBackbling()
   n6 = OR Boolean!(A=@n5, B=@n5.bIsBackblingACape, C:bool=@n5.bIsBackblingLong) -> (ReturnValue)
   n7 = Branch(execute=@n2.then_4, Condition=AND Boolean!(A=NOT Boolean(A=@n6), B=Get bIsSkyDiving(), C:bool=NOT Boolean(A=@n4)) -> (ReturnValue))
   > Set liftHairSkydive(liftHairSkydive:bool=true) -> (Output_Get:bool)
   Set liftHairSkydive(execute=@n7.else, liftHairSkydive=false)
   n8 = Branch(execute=@n2.then_3, Condition=Get bIsSkyDiving())
   > Set linearAccScale(linearAccScale:Vector="0.02,0.02,0.02") -> (Output_Get:Vector)
   Set linearAccScale(execute=@n8.else, linearAccScale="0.4,0.3,0.3")
   Set clavicle Up(
    execute=@n2.then_2
    clavicle Up:float=GetAlphaFromReferenceTransformDelta(SkeletalMeshComponent=Get BodyMesh(), BoneName=clavicle_r, BaseName=spine_05, InputRange=(X=0,Y=-30), TwistAxis="1,0,0")
   ) -> (Output_Get:float)
   Set head Rot Up(
    execute=@n2.then_1
    head Rot Up:float=GetAlphaFromReferenceTransformDelta(SkeletalMeshComponent=Get BodyMesh(), BoneName=head, BaseName=spine_05, InputRange=(X=0,Y=-30), TwistAxis="1,0,0")
   ) -> (Output_Get:float)
   n9 = Branch(execute=@n2.then_5, Condition=AND Boolean!(A=NOT Boolean(A=@n6), B=Get bIsParachuteOpened()) -> (ReturnValue))
   > Set liftHairParachute(liftHairParachute:bool=true) -> (Output_Get:bool)
   Set liftHairParachute(execute=@n9.else, liftHairParachute=false)
   Event BlueprintInitializeAnimation()
   > Set BodyMesh(BodyMesh=GetOwnerSkeletalMeshComponent())
   Event OnBeginSkydiving()
   Event OnEndSkydiving()
  ]}
]
variables [
 BodyMesh {type=SkeletalMeshComponent}
 clavicle Up {type=number}
 Gravity Override {type=Vector}
 Gravity Override Params {type=GravityOverrideParamsStruct}
 head Rot Up {type=number}
 liftHairParachute {type=boolean}
 liftHairSkydive {type=boolean}
 linearAccScale {type=Vector}
]
